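# Unlit translucent, two-sided material for a zap spline effect on the VinderTech apprentice glider, drawn on Niagara mesh particles with depth testing disabled.
# World position offset builds a camera-facing, tapered ribbon, adds a mid-span camera offset, and jitters it with scrolling noise that fades at both ends.
# Rotated foreground and background bands form the emissive, with particle dynamic parameters driving color variant, width and gradient reveal.
# Opacity combines particle alpha, the reveal mask and non-mobile depth fade.
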Material /Game/Gadgets/Assets/VinderTech_GliderChute/Glider_Apprentice/FX/M_Glider_Apprentice_ZapSpline
graph {
 properties {
  allowTranslucentCustomDepthWrites=true
  bDisableDepthTest=true
  bUsedWithNiagaraMeshParticles=true
  blendMode=BLEND_Translucent
  pixelDepthOffsetMode=PDOM_Legacy
  shadingModel=MSM_Unlit
  twoSided=true
  usedShadingModels=Unlit
 }
 nodes [
  n1 = OneMinus(None=ComponentMask(None=TextureCoordinate()) {g=true})
  n2 = DynamicParameter() {defaultValue=<rgba 1 1 0 0.5>, paramNames [Width, Offset, VariantPicker, Appear]}
  n3 = ScalarParameter() {defaultValue=1, group=02-Opacity, parameterName=Mask Size}
  n4 = StaticSwitchParameter(
   True=Frac(None=MaterialFunctionCall!(Gradient=@n1, Mask Position=@n2.Appear, Mask Size=@n3) {materialFunction=/Game/Effects/Fort_Effects/MaterialFunctions/MF_GradientMask}.Linear)
   False=Constant() {r=1}
  ) {defaultValue=true, group=02-Opacity, parameterName=Use Linear Reveal}
  n5 = CrossProduct(
   A=Subtract(A=CameraPositionWS(), B=WorldPosition().XYZ)
   B=MaterialFunctionCall() {materialFunction=&TangentBasis}.Y V
  )
  n6 = TextureCoordinate()
  n7 = ComponentMask(None=@n6) {g=true}
  n8 = SmoothStep(Value=@n7) {constMax=0.5}
  n9 = SmoothStep(Value=OneMinus(None=@n7)) {constMax=0.5}
  n10 = LinearInterpolate(
   A=ScalarParameter() {defaultValue=15, group=01-Spline, parameterName=BaseWidth, sortPriority=0}
   B=LinearInterpolate(
    A=ScalarParameter() {defaultValue=15, group=01-Spline, parameterName=TipsWidth, sortPriority=2}
    B=ScalarParameter() {defaultValue=15, group=01-Spline, parameterName=MiddleWidth, sortPriority=1}
    Alpha=@n8
   )
   Alpha=@n9
  )
  n11 = ComponentMask(None=Add(A=Constant() {r=-0.5}, B=@n6)) {r=true}
  n12 = Multiply(A=Multiply(A=@n4, B=Multiply(B=@n2.Width) {constA=1}) {constA=1}, B=Multiply(A=Normalize(VectorInput=@n5), B=Multiply(A=@n10, B=@n11)))
  n13 = Multiply(
   A=Multiply(
    A=MaterialFunctionCall() {materialFunction=&CameraDirectionVector}.Result
    B=ScalarParameter() {group=03-WPO, parameterName=CameraOffset}
   )
   B=Multiply(A=@n8, B=@n9)
  )
  n14 = ComponentMask(None=TextureCoordinate()) {g=true}
  n15 = Multiply(A=SmoothStep(Value=OneMinus(None=@n14)) {constMax=0.1}, B=SmoothStep(Value=@n14) {constMax=0.1})
  n16 = VectorParameter() {group=03-WPO, parameterName=UV_WPO_Noise, defaultValue=<rgba 1 1 0 0>}
  n17 = Add(
   A=Multiply(
    A=TextureCoordinate()
    B=MaterialFunctionCall(X=@n16.R, Y=@n16.R) {materialFunction=&MakeFloat2}.Result
   )
   B=Multiply(
    A=Time()
    B=MaterialFunctionCall(X=@n16.B, Y=@n16.B) {materialFunction=&MakeFloat2}.Result
   )
  )
  n18 = Multiply(
   A=ConstantBiasScale(None=TextureSample(UVs=@n17) {texture=/Game/TimeOfDay/LightFunction/T_Cloud2dNoise}.RGB) {bias=-0.5, scale=2}
   B=ScalarParameter() {group=03-WPO, parameterName=WPO_Strength}
  )
  Add(A=Add(A=Add(A=@n12, B=@n13), B=Multiply(A=@n15, B=@n18))) {constB=0}
  n19 = Constant() {r=1}
  n20 = MaterialFunctionCall!(Angle=ScalarParameter() {defaultValue=0.25, group=00-Color_Setup, parameterName=RotateBand}, UV/2DVector) {materialFunction=/Game/Characters/Player/Male/Medium/Bodies/M_MED_Faux/Materials/MF_2DRotate}
  n21 = MaterialFunctionCall!(
   Width=ScalarParameter() {defaultValue=0.5, group=00-Color_Setup, parameterName=BG_BandWidth}
   Sharpness=@n19
   Offset
   Direction Switch
   Compare
   Input Coordinates=@n20.Result
  ) {materialFunction=/Engine/Functions/Engine_MaterialFunctions03/Procedurals/GeneratedBand}
  n22 = DynamicParameter() {defaultValue=<rgba 1 1 0 0.5>, paramNames [Width, Offset, VariantPicker, Appear]}
  n23 = LinearInterpolate(
   A=VectorParameter() {group=00-Color_Setup, parameterName=BackgroundColor, defaultValue=<rgba 0 0.3005 0.2831 0>}.RGB
   B=VectorParameter() {group=00-Color_Setup, parameterName=VariantBackgroundColor, defaultValue=<rgba 0.625 0.0781 0.4671 1>}.RGB
   Alpha=@n22.VariantPicker
  )
  n24 = LinearInterpolate(
   A=VectorParameter() {group=00-Color_Setup, parameterName=ForegroundColor, defaultValue=<rgba 0.0052 1 1 0>}.RGB
   B=VectorParameter() {group=00-Color_Setup, parameterName=VariantForegroundColor, defaultValue=<rgba 1 0.3231 0.7991 1>}.RGB
   Alpha=@n22.VariantPicker
  )
  n25 = MaterialFunctionCall(Width=ScalarParameter() {defaultValue=0.3, group=00-Color_Setup, parameterName=FG_Bandwitdth}, Sharpness=@n19, Input Coordinates=@n20.Result) {materialFunction=&GeneratedBand}
  MaterialFunctionCall(Emissive=LinearInterpolate(A=Multiply(A=@n21.Result, B=@n23) {constB=0.05}, B=Multiply(A=@n24, B=@n25.Result), Alpha=@n25.Result)) {materialFunction=&MF_FX_EmissiveAdjustments}
  n26 = DepthFade(FadeDistance=ScalarParameter() {defaultValue=20, group=02-Opacity, parameterName=DepthFade})
  n27 = Constant() {r=1}
  Multiply(
   A=Multiply(A=ShadingPathSwitch(Default=@n26, Mobile=@n27) {desc=Migrated from FeatureLevelSwitch}, B=ParticleColor().A)
   B=Saturate(None=Ceil(None=@n4))
  )
  SmoothStep(Value=@n4) {constMax=0.2, constMin=0.2}
 ]
}
